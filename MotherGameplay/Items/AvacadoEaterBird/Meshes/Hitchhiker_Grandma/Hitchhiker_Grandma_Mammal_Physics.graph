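# Physics asset for the Hitchhiker_Grandma mammal mesh in the AvacadoEaterBird item folder, covering an alien's head, two arms and four eyestalks.
# Capsule bodies form three-segment chains for each arm and eyestalk, all branching from one large head body.
# Outer joints lock swing and twist, so each chain bends only at its head joint, where eyestalks have tighter swing limit values than arms.
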PhysicsAsset /MotherGameplay/Items/AvacadoEaterBird/Meshes/Hitchhiker_Grandma/Hitchhiker_Grandma_Mammal_Physics
bodies {
 alien_arm01_l [{shapeType=Capsule, length=7.0358, radius=6.1583, center=<4.5475 0 0>, rotation=<rot 90 0 0>}]
 alien_arm01_r [{shapeType=Capsule, length=7.4255, radius=7.0224, center=<4.3933 0 0>, rotation=<rot 90 0 0>}]
 alien_arm03_l [{shapeType=Capsule, length=8, radius=5, center=<4.3621 0 0>, rotation=<rot 90 0 0>}]
 alien_arm03_r [{shapeType=Capsule, length=5, radius=5, rotation=<rot 90 0 0>}]
 alien_arm05_l [{shapeType=Capsule, length=5, radius=5, rotation=<rot 90 0 0>}]
 alien_arm05_r [{shapeType=Capsule, length=5, radius=5, center=<-5.5053 0 0>, rotation=<rot 90 0 0>}]
 alien_eye02_bl [{shapeType=Capsule, length=5.75, radius=1.75, rotation=<rot 90 0 0>}]
 alien_eye02_br [{shapeType=Capsule, length=5.75, radius=1.75, rotation=<rot 90 0 0>}]
 alien_eye02_fl [{shapeType=Capsule, length=5.75, radius=1.75, rotation=<rot 90 0 0>}]
 alien_eye02_fr [{shapeType=Capsule, length=5.75, radius=1.75, rotation=<rot 90 0 0>}]
 alien_eye03_bl [{shapeType=Capsule, length=0, radius=3.5}]
 alien_eye03_br [{shapeType=Capsule, length=0, radius=3.5}]
 alien_eye03_fl [{shapeType=Capsule, length=0, radius=6}]
 alien_eye03_fr [{shapeType=Capsule, length=0, radius=6}]
 alien_eye_base_bl [{shapeType=Capsule, length=2, radius=2.25, center=<3.0624 -0.0001 0>, rotation=<rot 90 0 0>}]
 alien_eye_base_br [{shapeType=Capsule, length=2, radius=2.25, center=<3.2612 -0.0001 0>, rotation=<rot 90 0 0>}]
 alien_eye_base_fl [{shapeType=Capsule, length=2, radius=2.25, center=<4.1823 0.0001 0>, rotation=<rot 90 0 0>}]
 alien_eye_base_fr [{shapeType=Capsule, length=2, radius=2.25, center=<3.0365 -0.0001 0>, rotation=<rot 90 0 0>}]
 alien_head [{shapeType=Capsule, length=2, radius=18.25}]
}
constraints [
 {bone1Name=alien_arm01_l, bone2Name=alien_head, swing1LimitDegrees=30, swing2LimitDegrees=30, twistLimitDegrees=15}
 {bone1Name=alien_arm01_r, bone2Name=alien_head, swing1LimitDegrees=30, swing2LimitDegrees=30, twistLimitDegrees=15}
 {bone1Name=alien_eye_base_fl, bone2Name=alien_head, swing1LimitDegrees=10, swing2LimitDegrees=10, twistLimitDegrees=15}
 {bone1Name=alien_eye_base_fr, bone2Name=alien_head, swing1LimitDegrees=10, swing2LimitDegrees=10, twistLimitDegrees=15}
 {bone1Name=alien_eye_base_bl, bone2Name=alien_head, swing1LimitDegrees=10, swing2LimitDegrees=10, twistLimitDegrees=15}
 {bone1Name=alien_eye_base_br, bone2Name=alien_head, swing1LimitDegrees=10, swing2LimitDegrees=10, twistLimitDegrees=15}
 {
  bone1Name=alien_arm03_r
  bone2Name=alien_arm01_r
  swing1LimitDegrees=45
  swing1Motion=Locked
  swing2LimitDegrees=45
  swing2Motion=Locked
  twistLimitDegrees=45
  twistMotion=Locked
 }
 {
  bone1Name=alien_arm05_r
  bone2Name=alien_arm03_r
  swing1LimitDegrees=45
  swing1Motion=Locked
  swing2LimitDegrees=45
  swing2Motion=Locked
  twistLimitDegrees=45
  twistMotion=Locked
 }
 {
  bone1Name=alien_eye02_fl
  bone2Name=alien_eye_base_fl
  swing1LimitDegrees=45
  swing1Motion=Locked
  swing2LimitDegrees=45
  swing2Motion=Locked
  twistLimitDegrees=45
  twistMotion=Locked
 }
 {
  bone1Name=alien_eye03_fl
  bone2Name=alien_eye02_fl
  swing1LimitDegrees=45
  swing1Motion=Locked
  swing2LimitDegrees=45
  swing2Motion=Locked
  twistLimitDegrees=45
  twistMotion=Locked
 }
 {
  bone1Name=alien_arm03_l
  bone2Name=alien_arm01_l
  swing1LimitDegrees=45
  swing1Motion=Locked
  swing2LimitDegrees=45
  swing2Motion=Locked
  twistLimitDegrees=45
  twistMotion=Locked
 }
 {
  bone1Name=alien_arm05_l
  bone2Name=alien_arm03_l
  swing1LimitDegrees=45
  swing1Motion=Locked
  swing2LimitDegrees=45
  swing2Motion=Locked
  twistLimitDegrees=45
  twistMotion=Locked
 }
 {
  bone1Name=alien_eye02_fr
  bone2Name=alien_eye_base_fr
  swing1LimitDegrees=45
  swing1Motion=Locked
  swing2LimitDegrees=45
  swing2Motion=Locked
  twistLimitDegrees=45
  twistMotion=Locked
 }
 {
  bone1Name=alien_eye03_fr
  bone2Name=alien_eye02_fr
  swing1LimitDegrees=45
  swing1Motion=Locked
  swing2LimitDegrees=45
  swing2Motion=Locked
  twistLimitDegrees=45
  twistMotion=Locked
 }
 {
  bone1Name=alien_eye02_bl
  bone2Name=alien_eye_base_bl
  swing1LimitDegrees=45
  swing1Motion=Locked
  swing2LimitDegrees=45
  swing2Motion=Locked
  twistLimitDegrees=45
  twistMotion=Locked
 }
 {
  bone1Name=alien_eye03_bl
  bone2Name=alien_eye02_bl
  swing1LimitDegrees=45
  swing1Motion=Locked
  swing2LimitDegrees=45
  swing2Motion=Locked
  twistLimitDegrees=45
  twistMotion=Locked
 }
 {
  bone1Name=alien_eye02_br
  bone2Name=alien_eye_base_br
  swing1LimitDegrees=45
  swing1Motion=Locked
  swing2LimitDegrees=45
  swing2Motion=Locked
  twistLimitDegrees=45
  twistMotion=Locked
 }
 {
  bone1Name=alien_eye03_br
  bone2Name=alien_eye02_br
  swing1LimitDegrees=45
  swing1Motion=Locked
  swing2LimitDegrees=45
  swing2Motion=Locked
  twistLimitDegrees=45
  twistMotion=Locked
 }
]
