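# Animation Blueprint for the Grandma variant of the AvocadoEater creature, driving idle/run, intro, howl, bite, lunge, hit-react and death animations.
# A conduit-hub state machine switches states via bools set by named anim input events, while Speed selects idle or run.
# Its pose layers from alien_root over the attached parent's copied pose, then passes through AnimDynamics on eye bones, RigidBody simulation and Inertialization.
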AnimBlueprint /MotherGameplay/Items/AvacadoEaterBird/Animation/AvocadoEater_Grandma/AvocadoEater_Grandma_AnimBP
target_skeleton=/MotherGameplay/Items/AvacadoEaterBird/Meshes/Hitchhiker_Grandma/Hitchhiker_Grandma_Mammal_Skeleton
graphs [
 AnimGraph {nodes [
   n1 = Copy Pose From Mesh() {bCopyCurves=true, bUseAttachedParent=true}
   AnimGraphNode_StateMachine_0 = AvocadoEater_Grandma_Statemachine() {is=title AnimGraphNode_StateMachine}
   n2 = Anim Dynamics - Bone: alien_eye_base_fl(
    ComponentPose=Local To Component(LocalPose=Layered blend per bone(BasePose=@n1, BlendPoses_0=@AnimGraphNode_StateMachine_0) {layerSetup [{branchFilters [{boneName=alien_root}]}]})
    Alpha=1
    ExternalForce="0,0,0"
   ) {
    bGravityOverrideInSimSpace=true
    bUseGravityOverride=true
    alphaBoolBlend {blendOption=Linear}
    boundBone {boneName=alien_eye_base_fl}
    componentAppliedLinearAccClamp=<2000 2000 2000>
    componentLinearAccScale=<0.3 0.3 0.35>
    componentLinearVelScale=<0.3 0.3 0.35>
    gravityOverride {z=-980}
    physicsBodyDefinitions [{
      sphereCollisionRadius=0
      boundBone {boneName=alien_eye_base_fl}
      boxExtents=<5 5 5>
      constraintSetup {angularLimitsMax=<10 10 10>, angularLimitsMin=<0 0 0>, angularTarget=<0 0 0>}
      localJointOffset=<0.5 0 2.5>
     }]
    relativeSpaceBone {boneName=root}
   }
   n3 = Anim Dynamics - Bone: alien_eye01_fl(ComponentPose=@n2, Alpha=1, ExternalForce="0,0,0") {
    bGravityOverrideInSimSpace=true
    bUseGravityOverride=true
    alphaBoolBlend {blendOption=Linear}
    boundBone {boneName=alien_eye01_fl}
    componentAppliedLinearAccClamp=<2000 2000 2000>
    componentLinearAccScale=<0.3 0.3 0.35>
    componentLinearVelScale=<0.3 0.3 0.35>
    gravityOverride {z=-980}
    physicsBodyDefinitions [{
      sphereCollisionRadius=0
      boundBone {boneName=alien_eye01_fl}
      boxExtents=<5 5 5>
      constraintSetup {angularLimitsMax=<10 10 10>, angularLimitsMin=<0 0 0>, angularTarget=<0 0 0>}
      localJointOffset=<0.5 0 2.5>
     }]
    relativeSpaceBone {boneName=root}
   }
   n4 = Anim Dynamics - Bone: alien_eye_base_fr(ComponentPose=@n3, Alpha=1, ExternalForce="0,0,0") {
    bGravityOverrideInSimSpace=true
    bUseGravityOverride=true
    alphaBoolBlend {blendOption=Linear}
    boundBone {boneName=alien_eye_base_fr}
    componentAppliedLinearAccClamp=<2000 2000 2000>
    componentLinearAccScale=<0.3 0.3 0.35>
    componentLinearVelScale=<0.3 0.3 0.35>
    gravityOverride {z=-980}
    physicsBodyDefinitions [{
      sphereCollisionRadius=0
      boundBone {boneName=alien_eye_base_fr}
      boxExtents=<5 5 5>
      constraintSetup {angularLimitsMax=<10 10 10>, angularLimitsMin=<0 0 0>, angularTarget=<0 0 0>}
      localJointOffset=<0 0 2.5>
     }]
    relativeSpaceBone {boneName=root}
   }
   n5 = Anim Dynamics - Bone: alien_eye01_fr(ComponentPose=@n4, Alpha=1, ExternalForce="0,0,0") {
    bGravityOverrideInSimSpace=true
    bUseGravityOverride=true
    alphaBoolBlend {blendOption=Linear}
    boundBone {boneName=alien_eye01_fr}
    componentAppliedLinearAccClamp=<2000 2000 2000>
    componentLinearAccScale=<0.3 0.3 0.35>
    componentLinearVelScale=<0.3 0.3 0.35>
    gravityOverride {z=-980}
    physicsBodyDefinitions [{
      sphereCollisionRadius=0
      boundBone {boneName=alien_eye01_fr}
      boxExtents=<5 5 5>
      constraintSetup {angularLimitsMax=<10 10 10>, angularLimitsMin=<0 0 0>, angularTarget=<0 0 0>}
      localJointOffset=<0 0 2.5>
     }]
    relativeSpaceBone {boneName=root}
   }
   n6 = Anim Dynamics - Bone: alien_eye_base_bl(ComponentPose=@n5, Alpha=1, ExternalForce="0,0,0") {
    bGravityOverrideInSimSpace=true
    bUseGravityOverride=true
    alphaBoolBlend {blendOption=Linear}
    boundBone {boneName=alien_eye_base_bl}
    componentAppliedLinearAccClamp=<2000 2000 2000>
    componentLinearAccScale=<0.3 0.3 0.35>
    componentLinearVelScale=<0.3 0.3 0.35>
    gravityOverride {z=-980}
    physicsBodyDefinitions [{
      sphereCollisionRadius=0
      boundBone {boneName=alien_eye_base_bl}
      boxExtents=<5 5 5>
      constraintSetup {angularLimitsMax=<25 25 25>, angularLimitsMin=<0 0 0>, angularTarget=<0 0 0>}
      localJointOffset=<0 0 2.5>
     }]
    relativeSpaceBone {boneName=root}
   }
   n7 = Anim Dynamics - Bone: alien_eye01_bl(ComponentPose=@n6, Alpha=1, ExternalForce="0,0,0") {
    bGravityOverrideInSimSpace=true
    bUseGravityOverride=true
    alphaBoolBlend {blendOption=Linear}
    boundBone {boneName=alien_eye01_bl}
    componentAppliedLinearAccClamp=<2000 2000 2000>
    componentLinearAccScale=<0.3 0.3 0.35>
    componentLinearVelScale=<0.3 0.3 0.35>
    gravityOverride {z=-980}
    physicsBodyDefinitions [{
      sphereCollisionRadius=0
      boundBone {boneName=alien_eye01_bl}
      boxExtents=<5 5 5>
      constraintSetup {angularLimitsMax=<15 15 15>, angularLimitsMin=<0 0 0>, angularTarget=<0 0 0>}
      localJointOffset=<0 0 2.5>
     }]
    relativeSpaceBone {boneName=root}
   }
   n8 = Anim Dynamics - Bone: alien_eye_base_br(ComponentPose=@n7, Alpha=1, ExternalForce="0,0,0") {
    bGravityOverrideInSimSpace=true
    bUseGravityOverride=true
    alphaBoolBlend {blendOption=Linear}
    boundBone {boneName=alien_eye_base_br}
    componentAppliedLinearAccClamp=<2000 2000 2000>
    componentLinearAccScale=<0.3 0.3 0.35>
    componentLinearVelScale=<0.3 0.3 0.35>
    gravityOverride {z=-980}
    physicsBodyDefinitions [{
      sphereCollisionRadius=0
      boundBone {boneName=alien_eye_base_br}
      boxExtents=<5 5 5>
      constraintSetup {angularLimitsMax=<25 25 25>, angularLimitsMin=<0 0 0>, angularTarget=<0 0 0>}
      localJointOffset=<0 0 2.5>
     }]
    relativeSpaceBone {boneName=root}
   }
   n9 = Anim Dynamics - Bone: alien_eye01_br(ComponentPose=@n8, Alpha=1, ExternalForce="0,0,0") {
    bGravityOverrideInSimSpace=true
    bUseGravityOverride=true
    alphaBoolBlend {blendOption=Linear}
    boundBone {boneName=alien_eye01_br}
    componentAppliedLinearAccClamp=<2000 2000 2000>
    componentLinearAccScale=<0.3 0.3 0.35>
    componentLinearVelScale=<0.3 0.3 0.35>
    gravityOverride {z=-980}
    physicsBodyDefinitions [{
      sphereCollisionRadius=0
      boundBone {boneName=alien_eye01_br}
      boxExtents=<5 5 5>
      constraintSetup {angularLimitsMax=<15 15 15>, angularLimitsMin=<0 0 0>, angularTarget=<0 0 0>}
      localJointOffset=<0 0 2.5>
     }]
    relativeSpaceBone {boneName=root}
   }
   n10 = RigidBody(ComponentPose=@n9) {
    bOverrideWorldGravity=true
    overridePhysicsAsset=/MotherGameplay/Items/AvacadoEaterBird/Meshes/Hitchhiker_Grandma/Hitchhiker_Grandma_Mammal_Physics
    simulationSpace=BaseBoneSpace
    alphaBoolBlend {blendOption=Linear}
    baseBoneRef {boneName=root}
    componentAppliedLinearAccClamp=<2000 2000 2000>
    componentLinearAccScale=<0.3 0.3 0.35>
    componentLinearVelScale=<0.3 0.3 0.35>
    overrideWorldGravity {z=-980}
   }
   Output Pose(Result=Inertialization!(Source:Pose Link Structure=Component To Local(ComponentPose=@n10)) -> (Pose:Pose Link Structure) {is=title AnimGraphNode_Inertialization}) {layerGroup=None}
  ]}
 AvocadoEater_Grandma_Statemachine {
  entry=DirtyHackState
  path=AnimGraph.AnimGraphNode_StateMachine_0.AvocadoEater_Grandma_Statemachine
  ids=id
  nodes [
   AnimStateConduitNode_0 = Conduit()
   AnimStateNode_0 = HitReact() {is=title AnimStateNode}
   AnimStateNode_1 = Locomotion() {is=title AnimStateNode}
   AnimStateNode_2 = DirtyHackState() {is=title AnimStateNode}
   AnimStateNode_3 = Howl() {is=title AnimStateNode}
   AnimStateNode_4 = Death() {is=title AnimStateNode}
   AnimStateNode_5 = IntroIdle() {is=title AnimStateNode}
   AnimStateNode_6 = Bite() {is=title AnimStateNode}
   AnimStateNode_8 = Lunging() {is=title AnimStateNode}
   AnimStateTransitionNode_0 = AnimStateTransitionNode()
   AnimStateTransitionNode_1 = AnimStateTransitionNode() {crossfadeDuration=0.35, logicType=TLT_Inertialization, priorityOrder=100}
   AnimStateTransitionNode_10 = AnimStateTransitionNode()
   AnimStateTransitionNode_11 = AnimStateTransitionNode()
   AnimStateTransitionNode_12 = AnimStateTransitionNode() {crossfadeDuration=0.35, logicType=TLT_Inertialization, priorityOrder=4}
   AnimStateTransitionNode_13 = AnimStateTransitionNode()
   AnimStateTransitionNode_2 = AnimStateTransitionNode()
   AnimStateTransitionNode_3 = AnimStateTransitionNode() {logicType=TLT_Inertialization, priorityOrder=2}
   AnimStateTransitionNode_4 = AnimStateTransitionNode() {bAutomaticRuleBasedOnSequencePlayerInState=true}
   AnimStateTransitionNode_5 = AnimStateTransitionNode() {logicType=TLT_Inertialization, priorityOrder=6}
   AnimStateTransitionNode_6 = AnimStateTransitionNode() {logicType=TLT_Inertialization, priorityOrder=5}
   AnimStateTransitionNode_7 = AnimStateTransitionNode() {crossfadeDuration=0.35, logicType=TLT_Inertialization}
   AnimStateTransitionNode_8 = AnimStateTransitionNode()
   AnimStateTransitionNode_9 = AnimStateTransitionNode() {crossfadeDuration=0.35, logicType=TLT_Inertialization, priorityOrder=3}
  ]
  links [
   {
    dst=Conduit
    rule=AnimGraph.AnimGraphNode_StateMachine_0.AvocadoEater_Grandma_Statemachine.AnimStateTransitionNode_10.Transition
    src=Bite
    via=AnimStateTransitionNode_10
   }
   {
    dst=Bite
    rule=AnimGraph.AnimGraphNode_StateMachine_0.AvocadoEater_Grandma_Statemachine.AnimStateTransitionNode_9.Transition
    src=Conduit
    via=AnimStateTransitionNode_9
   }
   {
    dst=Death
    rule=AnimGraph.AnimGraphNode_StateMachine_0.AvocadoEater_Grandma_Statemachine.AnimStateTransitionNode_5.Transition
    src=Conduit
    via=AnimStateTransitionNode_5
   }
   {
    dst=HitReact
    rule=AnimGraph.AnimGraphNode_StateMachine_0.AvocadoEater_Grandma_Statemachine.AnimStateTransitionNode_6.Transition
    src=Conduit
    via=AnimStateTransitionNode_6
   }
   {
    dst=Howl
    rule=AnimGraph.AnimGraphNode_StateMachine_0.AvocadoEater_Grandma_Statemachine.AnimStateTransitionNode_3.Transition
    src=Conduit
    via=AnimStateTransitionNode_3
   }
   {
    dst=IntroIdle
    rule=AnimGraph.AnimGraphNode_StateMachine_0.AvocadoEater_Grandma_Statemachine.AnimStateTransitionNode_7.Transition
    src=Conduit
    via=AnimStateTransitionNode_7
   }
   {
    dst=Locomotion
    rule=AnimGraph.AnimGraphNode_StateMachine_0.AvocadoEater_Grandma_Statemachine.AnimStateTransitionNode_1.Transition
    src=Conduit
    via=AnimStateTransitionNode_1
   }
   {
    dst=Lunging
    rule=AnimGraph.AnimGraphNode_StateMachine_0.AvocadoEater_Grandma_Statemachine.AnimStateTransitionNode_12.Transition
    src=Conduit
    via=AnimStateTransitionNode_12
   }
   {
    dst=Conduit
    rule=AnimGraph.AnimGraphNode_StateMachine_0.AvocadoEater_Grandma_Statemachine.AnimStateTransitionNode_0.Transition
    src=DirtyHackState
    via=AnimStateTransitionNode_0
   }
   {
    dst=Conduit
    rule=AnimGraph.AnimGraphNode_StateMachine_0.AvocadoEater_Grandma_Statemachine.AnimStateTransitionNode_11.Transition
    src=HitReact
    via=AnimStateTransitionNode_11
   }
   {
    dst=Conduit
    rule=AnimGraph.AnimGraphNode_StateMachine_0.AvocadoEater_Grandma_Statemachine.AnimStateTransitionNode_4.Transition
    src=Howl
    via=AnimStateTransitionNode_4
   }
   {
    dst=Conduit
    rule=AnimGraph.AnimGraphNode_StateMachine_0.AvocadoEater_Grandma_Statemachine.AnimStateTransitionNode_8.Transition
    src=IntroIdle
    via=AnimStateTransitionNode_8
   }
   {
    dst=Conduit
    rule=AnimGraph.AnimGraphNode_StateMachine_0.AvocadoEater_Grandma_Statemachine.AnimStateTransitionNode_2.Transition
    src=Locomotion
    via=AnimStateTransitionNode_2
   }
   {
    dst=Conduit
    rule=AnimGraph.AnimGraphNode_StateMachine_0.AvocadoEater_Grandma_Statemachine.AnimStateTransitionNode_13.Transition
    src=Lunging
    via=AnimStateTransitionNode_13
   }
  ]
 }
 Conduit {
  path=AnimGraph.AnimGraphNode_StateMachine_0.AvocadoEater_Grandma_Statemachine.AnimStateConduitNode_0.Conduit
  nodes [Result(bCanEnterTransition=true)]
 }
 HitReact {
  path=AnimGraph.AnimGraphNode_StateMachine_0.AvocadoEater_Grandma_Statemachine.AnimStateNode_0.HitReact
  nodes [Output Animation Pose(Result=Sequence Player 'AvocadoEater_HitReact_FullBody'() {sequence=&/AvocadoEater_HitReact_FullBody})]
 }
 Locomotion {
  path=AnimGraph.AnimGraphNode_StateMachine_0.AvocadoEater_Grandma_Statemachine.AnimStateNode_1.Locomotion
  nodes [Output Animation Pose(Result=Blend Poses by bool(
     BlendPose_0=Sequence Player 'AvocadoEater_Grandma_Idle'() {sequence=&/AvocadoEater_Grandma_Idle}
     BlendPose_1=Sequence Player 'AvocadoEater_Grandma_Hitchiker_Run_v2'() {sequence=&/AvocadoEater_Grandma_Hitchiker_Run_v2}
     BlendTime_0=0.5
     BlendTime_1=0.5
     bActiveValue=Get IsIdle() -> (IsIdle:bool)
    ) {blendType=Linear, transitionType=Inertialization})]
 }
 DirtyHackState {
  path=AnimGraph.AnimGraphNode_StateMachine_0.AvocadoEater_Grandma_Statemachine.AnimStateNode_2.DirtyHackState
  nodes [Output Animation Pose(Result=Local Space Ref Pose())]
 }
 Howl {
  path=AnimGraph.AnimGraphNode_StateMachine_0.AvocadoEater_Grandma_Statemachine.AnimStateNode_3.Howl
  nodes [Output Animation Pose(Result=Sequence Player 'AvocadoEater_Grandma_Howl'() {bLoopAnimation=false, sequence=&/AvocadoEater_Grandma_Howl})]
 }
 Death {
  path=AnimGraph.AnimGraphNode_StateMachine_0.AvocadoEater_Grandma_Statemachine.AnimStateNode_4.Death
  nodes [Output Animation Pose(Result=Sequence Player 'AvocadoEater_Grandma_Death'() {bLoopAnimation=false, sequence=&/AvocadoEater_Grandma_Death})]
 }
 IntroIdle {
  path=AnimGraph.AnimGraphNode_StateMachine_0.AvocadoEater_Grandma_Statemachine.AnimStateNode_5.IntroIdle
  nodes [
   n1 = Sequence Player 'AvocadoEater_Grandma_Hitchiker_Into_Idle'() {
    bLoopAnimation=false
    sequence=&/AvocadoEater_Grandma_Hitchiker_Into_Idle
   }
   Output Animation Pose(Result=@n1)
  ]
 }
 Bite {
  path=AnimGraph.AnimGraphNode_StateMachine_0.AvocadoEater_Grandma_Statemachine.AnimStateNode_6.Bite
  nodes [
   n1 = Random Sequence Player!() -> (Pose:Pose Link Structure) {
    is=title AnimGraphNode_RandomPlayer
    entries [
     {
      chanceToPlay=0.5
      maxLoopCount=0
      maxPlayRate=1
      minLoopCount=0
      minPlayRate=1
      sequence=&/AvocadoEater_Grandma_Attack_Melee_Bite
      blendIn {blendOption=Linear, blendTime=0.2, customCurve=null}
     }
     {
      chanceToPlay=0.5
      maxLoopCount=0
      maxPlayRate=1
      minLoopCount=0
      minPlayRate=1
      sequence=&/AvocadoEater_Grandma_Attack_Melee_Bite_2
      blendIn {blendOption=Linear, blendTime=0.2, customCurve=null}
     }
    ]
   }
   Output Animation Pose(Result=@n1)
  ]
 }
 Lunging {
  path=AnimGraph.AnimGraphNode_StateMachine_0.AvocadoEater_Grandma_Statemachine.AnimStateNode_8.Lunging
  nodes [
   n1 = Sequence Player 'AvocadoEater_Grandma_Attack_Melee_Lunge'() {bLoopAnimation=false, sequence=&/AvocadoEater_Grandma_Attack_Melee_Lunge}
   Output Animation Pose(Result=@n1)
  ]
 }
 Transition {
  path=AnimGraph.AnimGraphNode_StateMachine_0.AvocadoEater_Grandma_Statemachine.AnimStateTransitionNode_0.Transition
  nodes [Result(bCanEnterTransition=true)]
 }
 Transition {
  path=AnimGraph.AnimGraphNode_StateMachine_0.AvocadoEater_Grandma_Statemachine.AnimStateTransitionNode_1.Transition
  nodes [Result(bCanEnterTransition=true)]
 }
 Transition {
  path=AnimGraph.AnimGraphNode_StateMachine_0.AvocadoEater_Grandma_Statemachine.AnimStateTransitionNode_10.Transition
  nodes [Result(bCanEnterTransition=NOT Boolean(A=Get IsBite() -> (IsBite:bool)))]
 }
 Transition {
  path=AnimGraph.AnimGraphNode_StateMachine_0.AvocadoEater_Grandma_Statemachine.AnimStateTransitionNode_11.Transition
  nodes [Result(bCanEnterTransition=NOT Boolean(A=Get IsHitReact() -> (IsHitReact:bool)))]
 }
 Transition {
  path=AnimGraph.AnimGraphNode_StateMachine_0.AvocadoEater_Grandma_Statemachine.AnimStateTransitionNode_12.Transition
  nodes [Result(bCanEnterTransition=Get IsLunging() -> (IsLunging:bool))]
 }
 Transition {
  path=AnimGraph.AnimGraphNode_StateMachine_0.AvocadoEater_Grandma_Statemachine.AnimStateTransitionNode_13.Transition
  nodes [Result(bCanEnterTransition=NOT Boolean(A=Get IsLunging()))]
 }
 Transition {
  path=AnimGraph.AnimGraphNode_StateMachine_0.AvocadoEater_Grandma_Statemachine.AnimStateTransitionNode_2.Transition
  nodes [Result(bCanEnterTransition=true)]
 }
 Transition {
  path=AnimGraph.AnimGraphNode_StateMachine_0.AvocadoEater_Grandma_Statemachine.AnimStateTransitionNode_3.Transition
  nodes [Result(bCanEnterTransition=Get IsHowling() -> (IsHowling:bool))]
 }
 Transition {
  path=AnimGraph.AnimGraphNode_StateMachine_0.AvocadoEater_Grandma_Statemachine.AnimStateTransitionNode_4.Transition
  nodes [Result(bCanEnterTransition=true)]
 }
 Transition {
  path=AnimGraph.AnimGraphNode_StateMachine_0.AvocadoEater_Grandma_Statemachine.AnimStateTransitionNode_5.Transition
  nodes [Result(bCanEnterTransition=Get IsDead() -> (IsDead:bool))]
 }
 Transition {
  path=AnimGraph.AnimGraphNode_StateMachine_0.AvocadoEater_Grandma_Statemachine.AnimStateTransitionNode_6.Transition
  nodes [Result(bCanEnterTransition=Get IsHitReact())]
 }
 Transition {
  path=AnimGraph.AnimGraphNode_StateMachine_0.AvocadoEater_Grandma_Statemachine.AnimStateTransitionNode_7.Transition
  nodes [Result(bCanEnterTransition=Get IsIntroIdle() -> (IsIntroIdle:bool))]
 }
 Transition {
  path=AnimGraph.AnimGraphNode_StateMachine_0.AvocadoEater_Grandma_Statemachine.AnimStateTransitionNode_8.Transition
  nodes [Result(bCanEnterTransition=NOT Boolean(A=Get IsIntroIdle()))]
 }
 Transition {
  path=AnimGraph.AnimGraphNode_StateMachine_0.AvocadoEater_Grandma_Statemachine.AnimStateTransitionNode_9.Transition
  nodes [Result(bCanEnterTransition=Get IsBite())]
 }
 EventGraph {nodes [
   Event BlueprintUpdateAnimation()
   > n1 = Branch(Condition="float < float"!(A:double=Get Speed() -> (Speed:float), B:double=300) -> (ReturnValue:bool) {is=title K2Node_CallFunction})
   > Set IsIdle(IsIdle:bool=true) -> (Output_Get:bool)
   Set IsIdle(execute=@n1.else, IsIdle=false)
   n2 = Event BlueprintProcessAnimInputEvent()
   > n3 = Branch(Condition=IsNamed(self=@n2.AnimInputEvent, InEventName=aie.Grandma.Howl))
   > n4 = Cast To FortAnimInputEvent_Bool!(execute, Object=@n2.AnimInputEvent) -> (then, CastFailed, AsFort Anim Input Event Bool:Fort Anim Input Event Bool Object Reference) {is=title K2Node_DynamicCast}
   > Set IsHowling(IsHowling:bool=Get bValue!(self:Fort Anim Input Event Bool Object Reference=@n4) -> (bValue:bool)) -> (Output_Get:bool)
   n5 = Branch(execute=@n3.else, Condition=IsNamed(self=@n2.AnimInputEvent, InEventName=aie.Grandma.Bite))
   > n6 = Cast To FortAnimInputEvent_Bool(Object=@n2.AnimInputEvent)
   > Set IsBite(IsBite:bool=Get bValue(self=@n6)) -> (Output_Get:bool)
   n7 = Branch(execute=@n5.else, Condition=IsNamed(self=@n2.AnimInputEvent, InEventName=aie.Grandma.Death))
   > n8 = Cast To FortAnimInputEvent_Bool(Object=@n2.AnimInputEvent)
   > Set IsDead(IsDead:bool=Get bValue(self=@n8)) -> (Output_Get:bool)
   n9 = Branch(execute=@n7.else, Condition=IsNamed(self=@n2.AnimInputEvent, InEventName=aie.Grandma.IntroIdle))
   > n10 = Cast To FortAnimInputEvent_Bool(Object=@n2.AnimInputEvent)
   > Set IsIntroIdle(IsIntroIdle:bool=Get bValue(self=@n10)) -> (Output_Get:bool)
   n11 = Branch(execute=@n9.else, Condition=IsNamed(self=@n2.AnimInputEvent, InEventName=aie.Grandma.Lunge))
   > n12 = Cast To FortAnimInputEvent_Bool(Object=@n2.AnimInputEvent)
   > Set IsLunging(IsLunging:bool=Get bValue(self=@n12)) -> (Output_Get:bool)
   Branch(execute=@n11.else, Condition=IsNamed(self=@n2.AnimInputEvent, InEventName=aie.Grandma.HitReact))
   > n13 = Cast To FortAnimInputEvent_Bool(Object=@n2.AnimInputEvent)
   > Set IsHitReact(IsHitReact:bool=Get bValue(self=@n13)) -> (Output_Get:bool)
  ]}
]
variables [
 IsBite {type=boolean}
 IsDead {type=boolean}
 IsHitReact {type=boolean}
 IsHowling {type=boolean}
 IsIdle {type=boolean}
 IsIntroIdle {type=boolean}
 IsLunging {type=boolean}
]
